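# Sound cue for a looping exterior windmill creak, played at low volume in the base ambient sound class.
# A quality-level node picks between two looping wave players: the standard wave for the first two levels and the "opt" variant for the last two.
# Attenuation is overridden in the cue with a custom falloff curve, stereo spread and listener focus that lowers volume when the source is out of focus.
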SoundCue /Game/Sounds/Fort_Ambient_Looping/Wind_Exterior/windmill_creaking_Cue
root=QualityLevel_0
nodes [
 {id=QualityLevel_0, children [WavePlayer_0,WavePlayer_0,WavePlayer_1,WavePlayer_1]}
 {id=WavePlayer_0, properties {bLooping=true, soundWaveAssetPtr=&/windmill_creaking}}
 {id=WavePlayer_1, properties {bLooping=true, soundWaveAssetPtr=&/windmill_creaking_opt}}
]
properties {
 bOverrideAttenuation=true
 maxDistance=3072
 soundClassObject=/Game/Sounds/FortSoundClassesAndModes/Fort_12_Base_Ambient
 volumeMultiplier=0.251
 attenuationOverrides {
  bEnableListenerFocus=true
  falloffDistance=2560
  focusAzimuth=25
  nonFocusAzimuth=45
  nonFocusVolumeAttenuation=0.4
  stereoSpread=500
  attenuationShapeExtents {x=512}
  customAttenuationCurve {editorCurveData {keys [<at 0 1>, {arriveTangent=-4.0197, interpMode=RCIM_Cubic, leaveTangent=-4.0197, tangentMode=RCTM_User, time=0.5, value=0.7}, <at 1 0>]}}
 }
}
